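# Sound class for in-game video audio in cinematics, a child of the RenderedCinematics class.
# It lowers volume, turns off reverb and routes its sounds to the Fortnite_Cinematics submix by default.
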SoundClass /Game/Sounds/FortSoundClassesAndModes/Cinematics_InGameVideo
parent=&/RenderedCinematics
properties {properties {bReverb=false, defaultSubmix=/Game/Sounds/SubMixes/Fortnite_Cinematics, volume=0.3}}
