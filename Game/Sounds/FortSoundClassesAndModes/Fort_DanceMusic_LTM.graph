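# Fort_DanceMusic_LTM is a sound class, named for dance music in a limited-time mode (LTM).
# It has no parent class and lists no child classes, so it stands outside any class hierarchy shown here.
# It turns off reverb for sounds assigned to it and sets a low-frequency effects bleed that sends part of the signal to the LFE channel.
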SoundClass /Game/Sounds/FortSoundClassesAndModes/Fort_DanceMusic_LTM
parent=null
properties {properties {bReverb=false, lFEBleed=0.25}}
