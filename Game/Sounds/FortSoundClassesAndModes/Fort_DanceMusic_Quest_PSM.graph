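# Sound class for quest dance music, named as a passive sound mix (PSM) trigger.
# While a sound of this class plays within the modifier's volume threshold range, it passively activates the Fort_DanceMusic_Quest_MixMod sound mix.
# The class has no parent, disables reverb and sets its own LFE bleed level.
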SoundClass /Game/Sounds/FortSoundClassesAndModes/Fort_DanceMusic_Quest_PSM
parent=null
properties {
 passiveSoundMixModifiers [{minVolumeThreshold=0.1, soundMix=&/Fort_DanceMusic_Quest_MixMod}]
 properties {bReverb=false, lFEBleed=0.25}
}
