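# Sound class that groups dance music audio in the Fortnite sound class hierarchy.
# It sits under the normal announcements class and parents four subclasses for first-person PSM, third-person, quest PSM and LTM dance music.
# Its own properties turn reverb off and set an LFE bleed level, leaving other settings at their defaults.
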SoundClass /Game/Sounds/FortSoundClassesAndModes/Fort_DanceMusic
parent=&/Fort_04_Normal_Announcements
children [
 &/Fort_DanceMusic_1P_PSM
 &/Fort_DanceMusic_3P
 &/Fort_DanceMusic_Quest_PSM
 &/Fort_DanceMusic_LTM
]
properties {properties {bReverb=false, lFEBleed=0.25}}
