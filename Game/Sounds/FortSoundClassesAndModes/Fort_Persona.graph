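# Fort_Persona is a SoundClass that sits under the Voice class and sets shared playback properties for persona voice audio.
# It boosts volume above unity, turns reverb off, applies ambient volumes and defaults output to the Fortnite_Dialogue submix.
# Volume modulation is enabled, but the class itself shows no modulator assignment.
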SoundClass /Game/Sounds/FortSoundClassesAndModes/Fort_Persona
parent=&/Voice
properties {properties {
  bApplyAmbientVolumes=true
  bReverb=false
  defaultSubmix=/Game/Sounds/SubMixes/Fortnite_Dialogue
  volume=1.5
  modulationSettings {volumeModulationDestination {bEnableModulation=true}}
 }}
