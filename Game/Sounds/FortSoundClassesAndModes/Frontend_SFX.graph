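# Sound class for front-end (menu) sound effects in the game's sound class hierarchy.
# It sits as a child of the SFX class and turns reverb off, so these sounds are not sent to reverb.
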SoundClass /Game/Sounds/FortSoundClassesAndModes/Frontend_SFX
parent=&SFX
properties {properties {bReverb=false}}
