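# Sound class for audio from concert devices, a child of the Fort_Concerts class.
# It applies ambient volume effects, disables reverb, and routes sounds to the Fortnite_SFX submix by default.
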SoundClass /Game/Sounds/FortSoundClassesAndModes/Fort_Concerts_Devices
parent=&/Fort_Concerts
properties {properties {bApplyAmbientVolumes=true, bReverb=false, defaultSubmix=&Fortnite_SFX}}
