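# Sound class for in-game concert audio, sitting under the events class in the Fortnite sound class hierarchy.
# It has one child class for concert devices.
# It lowers volume to roughly a third, applies ambient volumes, disables reverb and routes by default to the Fortnite SFX submix.
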SoundClass /Game/Sounds/FortSoundClassesAndModes/Fort_Concerts
parent=&Fort_Events
children [&/Fort_Concerts_Devices]
properties {properties {bApplyAmbientVolumes=true, bReverb=false, defaultSubmix=&Fortnite_SFX, volume=0.3509}}
