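# Sound class for extra-high-priority event audio, parented to the Fort_Events class.
# A passive sound mix modifier applies Fort_Events_ExtHigh_PSM automatically while sounds in this class play at or above a minimum volume threshold.
# Reverb is disabled for the class.
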SoundClass /Game/Sounds/FortSoundClassesAndModes/Fort_Events_Priority_ExtHigh
parent=&Fort_Events
properties {
 passiveSoundMixModifiers [{minVolumeThreshold=0.15, soundMix=/Game/Athena/Sounds/Events/SoundClasses/Fort_Events_ExtHigh_PSM}]
 properties {bReverb=false}
}
