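# Sound class for player footsteps heard from above or below, nested under the third-person Athena player foley class.
# It boosts volume above unity and disables reverb, with all other class properties left at defaults.
# The name suggests it separates vertically offset footsteps so they stand out in the mix.
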SoundClass /Game/Sounds/FortSoundClassesAndModes/Fort_Player_Footsteps_AboveOrBelow
parent=&/Fort_Player_Foley_3P_Athena
properties {properties {bReverb=false, volume=1.5}}
