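# Sound class for third-person player foley in Athena (Battle Royale), used for sounds other players make.
# It sits under the Athena player foley class and parents an above-or-below footsteps subclass.
# It boosts volume and disables reverb.
# It passively triggers an enemy-footstep sound mix whenever its volume passes a minimum threshold.
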SoundClass /Game/Sounds/FortSoundClassesAndModes/Fort_Player_Foley_3P_Athena
parent=&Fort_Player_Foley_Athena
children [&/Fort_Player_Footsteps_AboveOrBelow]
properties {
 passiveSoundMixModifiers [{minVolumeThreshold=0.095, soundMix=&/Fort_EnemyFootstep_MixMod_Athena}]
 properties {bReverb=false, volume=3}
}
