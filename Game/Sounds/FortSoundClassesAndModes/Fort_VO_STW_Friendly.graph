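# Sound class for friendly voice-over in the Save the World mode, judging by its name.
# It is a child of the Save the World NPC voice-over class, disables reverb and routes its sounds to the Fortnite dialogue submix by default.
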SoundClass /Game/Sounds/FortSoundClassesAndModes/Fort_VO_STW_Friendly
parent=&/Fort_VO_NPC_STW
properties {properties {bReverb=false, defaultSubmix=/Game/Sounds/SubMixes/Fortnite_Dialogue}}
