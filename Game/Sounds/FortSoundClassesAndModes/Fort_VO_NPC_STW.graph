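# Sound class for NPC voice lines in STW, a child of the general Fort_VO_NPC class.
# It groups two child classes that split those voices into friendly and enemy.
# It turns reverb off and sets the Fortnite_Dialogue submix as the default output.
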SoundClass /Game/Sounds/FortSoundClassesAndModes/Fort_VO_NPC_STW
parent=&Fort_VO_NPC
children [&/Fort_VO_STW_Friendly, &/Fort_VO_STW_Enemy]
properties {properties {bReverb=false, defaultSubmix=/Game/Sounds/SubMixes/Fortnite_Dialogue}}
